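# Material function that adds an animated glowing waveform line, a cipher signal, to the emissive of quest actor screens.
# A panning sine, tapered across UVs, modulated by hashed distortion and scaled by SignalControl custom primitive data, offsets a SphereMask that draws the line.
# A faint scrolling ramp, time-offset by actor position, joins it before view-facing falloff, tint and vertex-red masking.
# The panning gradient band nodes do not reach the output.
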MaterialFunction /BattlePassS23/QuestActors/Assets/Materials/MF_CipherSignal
graph {nodes [
  n1 = ComponentMask(None=StaticSwitchParameter(True=TextureCoordinate() {coordinateIndex=2}, False=TextureCoordinate() {coordinateIndex=1}) {group=Screens, parameterName="Use UV3 (3dsMax) "}) {g=true}
  n2 = ActorPositionWS()
  n3 = Add(
   A=Time()
   B=Divide(
    A=MaterialFunctionCall!(f2, f3=@n2, f4) {materialFunction=/Game/Packages/Fortnite_UniqueMaterials_1/VectorComponentOps}.f3 Add
    B=Constant() {r=32}
   )
  )
  n4 = LinearInterpolate(A=Add(A=@n3, B=Constant() {r=0.5}) {constB=0.5}, B=@n3, Alpha=VertexColor().R)
  n5 = Multiply(
   A=TextureSample(UVs=Add(A=@n1, B=@n4)) {samplerType=SAMPLERTYPE_LinearGrayscale, texture=/Game/Global/Textures/Default/Masks_and_Gradients/T_ScreenRamp_H}.R
   B=Constant() {r=0.2}
  )
  Add(A=@n5, B=Constant() {r=0.5})
  n6 = MaterialFunctionCall(Float2=NamedRerouteUsage()) {materialFunction=&BreakOutFloat2Components}
  n7 = Multiply(A=Time(), B=ScalarParameter() {defaultValue=0.2, parameterName=GradientPanSpeed})
  Add(A=@n6.G, B=@n7)
  Frac(None=Multiply(A=Time(), B=ScalarParameter() {defaultValue=1, parameterName=SignalDistortion}))
  n8 = GetMaterialAttributes(None=FunctionInput() {inputName=Material, inputType=FunctionInput_MaterialAttributes}) {attributeGetTypes [#1]}
  n9 = Saturate(None=DotProduct(A=CameraVectorWS(), B=PixelNormalWS()))
  n10 = MaterialFunctionCall(Float2=NamedRerouteDeclaration(None=TextureCoordinate() {coordinateIndex=1}) {name=UVCoords, nodeColor=<rgba 0 1 0.2031 1>}) {materialFunction=&BreakOutFloat2Components}
  n11 = Add(A=@n10.R) {constB=-0.25}
  n12 = Time()
  n13 = ScalarParameter() {defaultValue=0.2, parameterName=FrequencyPanSpeed}
  n14 = Multiply(
   A=Add(A=@n10.R, B=Multiply(A=@n12, B=@n13))
   B=NamedRerouteDeclaration(None=ScalarParameter() {defaultValue=4, parameterName=WaveFrequency}) {name=WaveFrequency, nodeColor=<rgba 0.1016 0 1 1>}
  )
  n15 = Multiply(
   A=ScalarParameter() {defaultValue=0.25, parameterName=FrequencyStr}
   B=LinearInterpolate(B=NamedRerouteUsage(), Alpha=ScalarParameter() {parameterName=Distortio0nAmt}) {constA=1}
  )
  n16 = ScalarParameter() {bUseCustomPrimitiveData=true, defaultValue=1, parameterName=SignalControl}
  n17 = Multiply(A=ConstantBiasScale(None=Sine(None=@n11)), B=Multiply(A=Sine(None=@n14), B=Multiply(A=@n15, B=@n16)))
  n18 = Add(
   A=SphereMask(
    A=Add(A=@n17, B=@n10.G)
    B=ScalarParameter() {defaultValue=0.5, parameterName=Position}
    Radius=ScalarParameter() {defaultValue=0.025, parameterName=SignalLineRadius}
    Hardness=ScalarParameter() {parameterName=LineHardness}
   )
   B=@n5
  )
  n19 = Add(
   A=@n8.EmissiveColor
   B=Multiply(A=VectorParameter() {parameterName=SignalLineColor, defaultValue=<rgba 0.6793 0.9531 0.1291 0>}.RGB, B=Multiply(A=@n9, B=@n18))
  )
  n20 = VertexColor()
  FunctionOutput(None=SetMaterialAttributes!(MaterialAttributes=@n8.MaterialAttributes, EmissiveColor=LinearInterpolate(A=@n8.EmissiveColor, B=@n19, Alpha=@n20.R)) {attributeSetTypes [#1]})
  n21 = MaterialFunctionCall(Float2=NamedRerouteUsage()) {materialFunction=&BreakOutFloat2Components}
  n22 = NamedRerouteUsage()
  n23 = Multiply(A=@n22) {constB=0.5}
  n24 = Sine(None=Add(A=Multiply(A=@n21.R, B=@n23)) {constB=-0.25})
  n25 = Multiply(A=@n22, B=@n21.R)
  n26 = Constant() {r=1}
  n27 = MaterialFunctionCall(Seed=ScalarParameter() {defaultValue=2, parameterName=Seed}, Vec2=Floor(None=AppendVector(A=@n25, B=@n26))) {materialFunction=&MF_Hash}
  n28 = Add(
   A=ConstantBiasScale(None=ComponentMask(None=@n27.Vec2) {r=true})
   B=Multiply(A=Time(), B=ScalarParameter() {defaultValue=0.2, parameterName=FrequencyWaveSpeed})
  )
  NamedRerouteDeclaration(None=Multiply(A=ConstantBiasScale(None=@n24), B=Sine(None=@n28))) {name=FrequencyDistortion, nodeColor=<rgba 0.1484 0 1 1>}
  ScalarParameter() {defaultValue=0.5, parameterName=GradientPosition}
  SphereMask(
   A=@n6.G
   B=Frac(None=@n7)
   Radius=ScalarParameter() {defaultValue=0.2, parameterName=GradientSignalLineRadius}
   Hardness=ScalarParameter() {parameterName=GradientLineHardness}
  )
 ]}
